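# Physics asset for the female medium Rave Ninja body's arm-area secondary motion, covering layered pauldrons, zipper pulls and a chain.
# Each pauldron segment hangs from a master body with free rotation to spine_04, tightly limited except the looser lowest one.
# Zipper boxes constrain to three pauldron segments, and a five-link capsule chain hangs from the lowest.
# Large spine and side-belt capsules plus a left forearm capsule appear to act as colliders.
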PhysicsAsset /BRCosmetics/Characters/Player/Female/Medium/Bodies/F_MED_Rave_Ninja/Meshes/F_MED_Rave_Ninja_Physics_Arm
bodies {
 dyn_beltSideA_l [{
   shapeType=Capsule
   length=5
   radius=12.7373
   shapeName=dyn_beltSideA_l_capsule
   center=<-8.4521 2.369 -8.5314>
   rotation=<rot -69.9999 0.0006 -0.0005>
  }]
 dyn_chain1 [{shapeType=Capsule, length=1, radius=1, shapeName=dyn_chain1_capsule, center=<0.3 0 0>, rotation=<rot 0 90 90>}]
 dyn_chain2 [{shapeType=Capsule, length=1, radius=0.9, shapeName=dyn_chain2_capsule, center=<1.2929 0 0>, rotation=<rot 0 90 90>}]
 dyn_chain3 [{shapeType=Capsule, length=1, radius=1, shapeName=dyn_chain3_capsule, center=<2.5455 0 0>, rotation=<rot 0 90 90>}]
 dyn_chain4 [{shapeType=Capsule, length=1, radius=1, shapeName=dyn_chain4_capsule, center=<4.0954 0 0>, rotation=<rot 0 90 90>}]
 dyn_chain5 [
  {shapeType=Capsule, length=2, radius=1, shapeName=dyn_chain5_capsule, center=<5.8921 0 0>, rotation=<rot 0 90 90>}
  {shapeType=Capsule, length=0, radius=2, shapeName=dyn_chain5_capsule, center=<5.086 0 -0.5782>, rotation=<rot 0 0 90>}
 ]
 dyn_pauldronLower [{shapeType=Capsule, length=5, radius=3.5, center=<4.7712 0.3196 2.6323>, rotation=<rot 0 0 -90>}]
 dyn_pauldronLowerMid [{shapeType=Capsule, length=5, radius=7, center=<6.4162 -0.0074 4.0963>, rotation=<rot 0 0 -90>}]
 dyn_pauldronLowerMidMstr [{shapeType=Capsule, length=0.505, radius=0.505, rotation=<rot 0 -90 0>}]
 dyn_pauldronLowerMstr [{shapeType=Capsule, length=0.505, radius=0.505, rotation=<rot 0 -90 0>}]
 dyn_pauldronUpper [{shapeType=Capsule, length=5, radius=7.5, center=<5.216 0.2511 3.138>, rotation=<rot 0 0 -90>}]
 dyn_pauldronUpperMid [{shapeType=Capsule, length=5, radius=7, center=<6.8827 0.1251 4.324>, rotation=<rot 0 0 -90>}]
 dyn_pauldronUpperMidMstr [{shapeType=Capsule, length=0.505, radius=0.505, rotation=<rot 0 -90 0>}]
 dyn_pauldronUpperMstr [{shapeType=Capsule, length=0.505, radius=0.505, rotation=<rot 0 -90 0>}]
 dyn_zipperLower2 [{shapeType=Box, extentX=3.4668, extentY=1.4766, extentZ=0.4576, center=<1.2721 0.0312 -0.0225>}]
 dyn_zipperMid2 [{shapeType=Box, extentX=3.51, extentY=1.5, extentZ=0.5, center=<1.2738 0.0079 0.0295>}]
 dyn_zipperUpper2 [{shapeType=Box, extentX=3.5193, extentY=1.4415, extentZ=0.4246, center=<1.3042 0.0173 -0.0219>}]
 lowerarm_twist_02_l [{shapeType=Capsule, length=15, radius=4, rotation=<rot 90 0 0>}]
 spine_04 [{shapeType=Capsule, length=0, radius=14.8209, shapeName=spine_04_capsule, center=<-5.4259 1.2549 -0.1721>, rotation=<rot 0 90 90>}]
}
constraints [
 {
  bone1Name=dyn_pauldronUpper
  bone2Name=dyn_pauldronUpperMstr
  swing1LimitDegrees=0
  swing1Motion=Locked
  swing2LimitDegrees=3
  twistLimitDegrees=2
 }
 {bone1Name=dyn_zipperUpper2, bone2Name=dyn_pauldronUpperMid, swing1LimitDegrees=45, swing2LimitDegrees=45, twistLimitDegrees=15}
 {
  bone1Name=dyn_pauldronUpperMid
  bone2Name=dyn_pauldronUpperMidMstr
  swing1LimitDegrees=0
  swing1Motion=Locked
  swing2LimitDegrees=3
  twistLimitDegrees=2
 }
 {bone1Name=dyn_zipperMid2, bone2Name=dyn_pauldronLowerMid, swing1LimitDegrees=45, swing2LimitDegrees=45, twistLimitDegrees=15}
 {
  bone1Name=dyn_pauldronLowerMid
  bone2Name=dyn_pauldronLowerMidMstr
  swing1LimitDegrees=0
  swing1Motion=Locked
  swing2LimitDegrees=3
  twistLimitDegrees=2
 }
 {bone1Name=dyn_zipperLower2, bone2Name=dyn_pauldronLower, swing1LimitDegrees=45, swing2LimitDegrees=45, twistLimitDegrees=15}
 {bone1Name=dyn_pauldronLower, bone2Name=dyn_pauldronLowerMstr, swing1LimitDegrees=45, swing2LimitDegrees=45, twistLimitDegrees=15}
 {bone1Name=dyn_chain1, bone2Name=dyn_pauldronLower, swing1LimitDegrees=5, swing2LimitDegrees=5, twistLimitDegrees=1}
 {bone1Name=dyn_chain2, bone2Name=dyn_chain1, swing1LimitDegrees=20, swing2LimitDegrees=10, twistLimitDegrees=5}
 {bone1Name=dyn_chain3, bone2Name=dyn_chain2, swing1LimitDegrees=20, swing2LimitDegrees=10, twistLimitDegrees=5}
 {bone1Name=dyn_chain4, bone2Name=dyn_chain3, swing1LimitDegrees=20, swing2LimitDegrees=10, twistLimitDegrees=5}
 {bone1Name=dyn_chain5, bone2Name=dyn_chain4, swing1LimitDegrees=20, swing2LimitDegrees=10, twistLimitDegrees=5}
 {
  bone1Name=lowerarm_twist_02_l
  bone2Name=spine_04
  swing1LimitDegrees=45
  swing1Motion=Free
  swing2LimitDegrees=45
  swing2Motion=Free
  twistLimitDegrees=45
  twistMotion=Free
 }
 {
  bone1Name=dyn_pauldronUpperMstr
  bone2Name=spine_04
  swing1LimitDegrees=45
  swing1Motion=Free
  swing2LimitDegrees=45
  swing2Motion=Free
  twistLimitDegrees=45
  twistMotion=Free
 }
 {
  bone1Name=dyn_pauldronUpperMidMstr
  bone2Name=spine_04
  swing1LimitDegrees=45
  swing1Motion=Free
  swing2LimitDegrees=45
  swing2Motion=Free
  twistLimitDegrees=45
  twistMotion=Free
 }
 {
  bone1Name=dyn_pauldronLowerMidMstr
  bone2Name=spine_04
  swing1LimitDegrees=45
  swing1Motion=Free
  swing2LimitDegrees=45
  swing2Motion=Free
  twistLimitDegrees=45
  twistMotion=Free
 }
 {
  bone1Name=dyn_pauldronLowerMstr
  bone2Name=spine_04
  swing1LimitDegrees=45
  swing1Motion=Free
  swing2LimitDegrees=45
  swing2Motion=Free
  twistLimitDegrees=45
  twistMotion=Free
 }
]
