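# Physics asset for the belt of the F_MED_Rave_Ninja female player body, simulating a small bone chain that hangs off the pelvis.
# Three capsule bodies cover the pelvis, a tiny dyn_beltBaseMstr link and the larger dyn_beltBase.
# The master link rotates freely on all angular axes relative to the pelvis.
# The belt base locks twist and swing1 to that link, leaving only a narrow swing2 range.
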PhysicsAsset /BRCosmetics/Characters/Player/Female/Medium/Bodies/F_MED_Rave_Ninja/Meshes/F_MED_Rave_Ninja_Physics_Belt
bodies {
 dyn_beltBase [{shapeType=Capsule, length=11, radius=3, center=<-2.4 0 0>, rotation=<rot 90 0 0>}]
 dyn_beltBaseMstr [{shapeType=Capsule, length=0.505, radius=0.505}]
 pelvis [{shapeType=Capsule, length=1.3, radius=5.8, center=<0.1701 -1.6972 -0.3205>}]
}
constraints [
 {
  bone1Name=dyn_beltBaseMstr
  bone2Name=pelvis
  swing1LimitDegrees=45
  swing1Motion=Free
  swing2LimitDegrees=45
  swing2Motion=Free
  twistLimitDegrees=45
  twistMotion=Free
 }
 {
  bone1Name=dyn_beltBase
  bone2Name=dyn_beltBaseMstr
  swing1LimitDegrees=5
  swing1Motion=Locked
  swing2LimitDegrees=10
  twistLimitDegrees=5
  twistMotion=Locked
 }
]
