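# Physics asset for the jacket of the female medium Rave Ninja player cosmetic, defining bodies and joints for its dynamic jacket bones.
# Thin boxes model the panels, and capsules form the pelvis and spine.
# Front panels form two-segment chains per side from small master capsules with free rotation on spine_04; side and back-mid panels also attach there, the center back to spine_05.
# Panel joints limit swing and keep twist tight.
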PhysicsAsset /BRCosmetics/Characters/Player/Female/Medium/Bodies/F_MED_Rave_Ninja/Meshes/F_MED_Rave_Ninja_Physics_Jacket
bodies {
 dyn_jacketBack [{shapeType=Box, extentX=8, extentY=6, extentZ=1, center=<4 0 0>}]
 dyn_jacketBackMid_l [{shapeType=Box, extentX=6, extentY=6, extentZ=1, center=<3 0 0>}]
 dyn_jacketBackMid_r [{shapeType=Box, extentX=6, extentY=6, extentZ=1, center=<3 0 0>}]
 dyn_jacketFront1_l [{shapeType=Box, extentX=8, extentY=8, extentZ=1, center=<4 0 0>}]
 dyn_jacketFront1_r [{shapeType=Box, extentX=8, extentY=8, extentZ=1, center=<4 0 0>}]
 dyn_jacketFront2_l [{shapeType=Box, extentX=6, extentY=8, extentZ=1, center=<3 0 0>}]
 dyn_jacketFront2_r [{shapeType=Box, extentX=6, extentY=8, extentZ=1, center=<3 0 0>}]
 dyn_jacketFrontMid1_l [{shapeType=Box, extentX=8, extentY=6, extentZ=1, center=<4 0 0>}]
 dyn_jacketFrontMid1_r [{shapeType=Box, extentX=8, extentY=6, extentZ=1, center=<4 0 0>}]
 dyn_jacketFrontMid2_l [{shapeType=Box, extentX=8, extentY=5, extentZ=1, center=<4 0 0>}]
 dyn_jacketFrontMid2_r [{shapeType=Box, extentX=8, extentY=5, extentZ=1, center=<4 0 0>}]
 dyn_jacketFrontMidMstr_l [{shapeType=Capsule, length=0.505, radius=0.505}]
 dyn_jacketFrontMidMstr_r [{shapeType=Capsule, length=0.505, radius=0.505}]
 dyn_jacketFrontMstr_l [{shapeType=Capsule, length=0.505, radius=0.505}]
 dyn_jacketFrontMstr_r [{shapeType=Capsule, length=0.505, radius=0.505}]
 dyn_jacketSide_l [{shapeType=Box, extentX=6, extentY=8, extentZ=1, center=<3 0 0>}]
 dyn_jacketSide_r [{shapeType=Box, extentX=6, extentY=8, extentZ=1, center=<3 0 0>}]
 pelvis [{shapeType=Capsule, length=10, radius=10}]
 spine_02 [{shapeType=Capsule, length=10, radius=8}]
 spine_03 [{shapeType=Capsule, length=10, radius=8}]
 spine_04 [{shapeType=Capsule, length=10, radius=8}]
 spine_05 [{shapeType=Capsule, length=10, radius=8, rotation=<rot 0 0 180>}]
}
constraints [
 {bone1Name=spine_02, bone2Name=pelvis, swing1LimitDegrees=45, swing2LimitDegrees=45, twistLimitDegrees=45}
 {bone1Name=spine_04, bone2Name=spine_03, swing1LimitDegrees=45, swing2LimitDegrees=45, twistLimitDegrees=45}
 {
  bone1Name=spine_03
  bone2Name=spine_02
  swing1LimitDegrees=45
  swing1Motion=Free
  swing2LimitDegrees=45
  swing2Motion=Free
  twistLimitDegrees=45
  twistMotion=Free
 }
 {bone1Name=dyn_jacketBack, bone2Name=spine_05, swing1LimitDegrees=45, swing2LimitDegrees=60, twistLimitDegrees=5}
 {
  bone1Name=dyn_jacketFrontMstr_l
  bone2Name=spine_04
  swing1LimitDegrees=45
  swing1Motion=Free
  swing2LimitDegrees=45
  swing2Motion=Free
  twistLimitDegrees=45
  twistMotion=Free
 }
 {bone1Name=dyn_jacketFront1_l, bone2Name=dyn_jacketFrontMstr_l, swing1LimitDegrees=45, swing2LimitDegrees=60, twistLimitDegrees=5}
 {bone1Name=dyn_jacketFront2_l, bone2Name=dyn_jacketFront1_l, swing1LimitDegrees=45, swing2LimitDegrees=60, twistLimitDegrees=5}
 {
  bone1Name=dyn_jacketFrontMidMstr_l
  bone2Name=spine_04
  swing1LimitDegrees=45
  swing1Motion=Free
  swing2LimitDegrees=45
  swing2Motion=Free
  twistLimitDegrees=45
  twistMotion=Free
 }
 {bone1Name=dyn_jacketFrontMid1_l, bone2Name=dyn_jacketFrontMidMstr_l, swing1LimitDegrees=45, swing2LimitDegrees=60, twistLimitDegrees=5}
 {bone1Name=dyn_jacketFrontMid2_l, bone2Name=dyn_jacketFrontMid1_l, swing1LimitDegrees=45, swing2LimitDegrees=60, twistLimitDegrees=5}
 {bone1Name=dyn_jacketSide_l, bone2Name=spine_04, swing1LimitDegrees=45, swing2LimitDegrees=60, twistLimitDegrees=5}
 {bone1Name=dyn_jacketBackMid_l, bone2Name=spine_04, swing1LimitDegrees=45, swing2LimitDegrees=60, twistLimitDegrees=5}
 {
  bone1Name=dyn_jacketFrontMstr_r
  bone2Name=spine_04
  swing1LimitDegrees=45
  swing1Motion=Free
  swing2LimitDegrees=45
  swing2Motion=Free
  twistLimitDegrees=45
  twistMotion=Free
 }
 {bone1Name=dyn_jacketFront1_r, bone2Name=dyn_jacketFrontMstr_r, swing1LimitDegrees=45, swing2LimitDegrees=60, twistLimitDegrees=5}
 {bone1Name=dyn_jacketFront2_r, bone2Name=dyn_jacketFront1_r, swing1LimitDegrees=45, swing2LimitDegrees=60, twistLimitDegrees=5}
 {
  bone1Name=dyn_jacketFrontMidMstr_r
  bone2Name=spine_04
  swing1LimitDegrees=45
  swing1Motion=Free
  swing2LimitDegrees=45
  swing2Motion=Free
  twistLimitDegrees=45
  twistMotion=Free
 }
 {bone1Name=dyn_jacketFrontMid1_r, bone2Name=dyn_jacketFrontMidMstr_r, swing1LimitDegrees=45, swing2LimitDegrees=60, twistLimitDegrees=5}
 {bone1Name=dyn_jacketFrontMid2_r, bone2Name=dyn_jacketFrontMid1_r, swing1LimitDegrees=45, swing2LimitDegrees=60, twistLimitDegrees=5}
 {bone1Name=dyn_jacketSide_r, bone2Name=spine_04, swing1LimitDegrees=45, swing2LimitDegrees=60, twistLimitDegrees=5}
 {bone1Name=dyn_jacketBackMid_r, bone2Name=spine_04, swing1LimitDegrees=45, swing2LimitDegrees=60, twistLimitDegrees=5}
 {
  bone1Name=spine_05
  bone2Name=spine_04
  swing1LimitDegrees=45
  swing1Motion=Free
  swing2LimitDegrees=45
  swing2Motion=Free
  twistLimitDegrees=45
  twistMotion=Free
 }
]
