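# Physics asset for the right leg of the female medium Rave Ninja cosmetic body, covering its dangling lower belt and right boot tab.
# A seven-bone belt chain of small, gradually growing capsules and a two-bone boot tab both hang from a thigh capsule body.
# Each chain's small box root joins the thigh with free swing and twist, while the links below use narrow swing and tighter twist limits.
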PhysicsAsset /BRCosmetics/Characters/Player/Female/Medium/Bodies/F_MED_Rave_Ninja/Meshes/F_MED_Rave_Ninja_Physics_Leg_R
bodies {
 dyn_beltLower1 [{shapeType=Box, extentX=1.01, extentY=1.01, extentZ=1.01}]
 dyn_beltLower2 [{shapeType=Capsule, length=0.5, radius=1, center=<1 0 0>, rotation=<rot 90 0 0>}]
 dyn_beltLower3 [{shapeType=Capsule, length=1, radius=1, center=<1 0 0>, rotation=<rot 90 0 0>}]
 dyn_beltLower4 [{shapeType=Capsule, length=1.5, radius=1.5, center=<2 0 0>, rotation=<rot 90 0 0>}]
 dyn_beltLower5 [{shapeType=Capsule, length=1.5, radius=1.5, center=<2 0 0>, rotation=<rot 90 0 0>}]
 dyn_beltLower6 [{shapeType=Capsule, length=1.5, radius=1.5, center=<2 0 0>, rotation=<rot 90 0 0>}]
 dyn_beltLower7 [{shapeType=Capsule, length=2, radius=1.5, center=<2 0 0>, rotation=<rot 90 0 0>}]
 dyn_bootTabMstr_r [{shapeType=Box, extentX=1.01, extentY=1.01, extentZ=1.01}]
 dyn_bootTab_r [{shapeType=Box, extentX=2.3452, extentY=1.7467, extentZ=2.1703, center=<2.0237 -0.0376 -0.006>}]
 thigh_r [{shapeType=Capsule, length=33.5176, radius=9.1946, center=<23.1245 1.4324 0.0907>, rotation=<rot 90 0 0>}]
}
constraints [
 {
  bone1Name=dyn_beltLower1
  bone2Name=thigh_r
  swing1LimitDegrees=45
  swing1Motion=Free
  swing2LimitDegrees=45
  swing2Motion=Free
  twistLimitDegrees=45
  twistMotion=Free
 }
 {bone1Name=dyn_beltLower2, bone2Name=dyn_beltLower1, swing1LimitDegrees=20, swing2LimitDegrees=20, twistLimitDegrees=5}
 {bone1Name=dyn_beltLower3, bone2Name=dyn_beltLower2, swing1LimitDegrees=20, swing2LimitDegrees=20, twistLimitDegrees=5}
 {bone1Name=dyn_beltLower4, bone2Name=dyn_beltLower3, swing1LimitDegrees=20, swing2LimitDegrees=20, twistLimitDegrees=5}
 {bone1Name=dyn_beltLower5, bone2Name=dyn_beltLower4, swing1LimitDegrees=20, swing2LimitDegrees=20, twistLimitDegrees=5}
 {bone1Name=dyn_beltLower6, bone2Name=dyn_beltLower5, swing1LimitDegrees=20, swing2LimitDegrees=20, twistLimitDegrees=5}
 {bone1Name=dyn_beltLower7, bone2Name=dyn_beltLower6, swing1LimitDegrees=20, swing2LimitDegrees=20, twistLimitDegrees=5}
 {
  bone1Name=dyn_bootTabMstr_r
  bone2Name=thigh_r
  swing1LimitDegrees=45
  swing1Motion=Free
  swing2LimitDegrees=45
  swing2Motion=Free
  twistLimitDegrees=45
  twistMotion=Free
 }
 {bone1Name=dyn_bootTab_r, bone2Name=dyn_bootTabMstr_r, swing1LimitDegrees=20, swing2LimitDegrees=20, twistLimitDegrees=5}
]
