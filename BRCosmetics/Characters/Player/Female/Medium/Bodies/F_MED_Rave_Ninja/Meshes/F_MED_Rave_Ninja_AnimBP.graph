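# Post-process-style cosmetic AnimBP for the female Rave Ninja body that copies the attached parent's pose and adds secondary motion.
# Bone constraints feed chained RigidBody simulations for right leg, arm, jacket and belt.
# Crouching offsets belt bones and disables belt physics; LOD1 adjusts neck clothes and head.
# The update event sets emote-dependent alpha and arm gravity, whose skydive logic is then overwritten by a deprecated gravity-override function.
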AnimBlueprint /BRCosmetics/Characters/Player/Female/Medium/Bodies/F_MED_Rave_Ninja/Meshes/F_MED_Rave_Ninja_AnimBP
target_skeleton=&/F_MED_Rave_Ninja_Skeleton
graphs [
 AnimGraph {nodes [
   n1 = "Constraint - thigh_l_ol (RR)"!(
    ComponentPose=Local To Component(LocalPose=Copy Pose From Mesh() {bUseAttachedParent=true})
    Alpha=1
    ConstraintWeights_0=0.5
    ConstraintWeights_1:float=0.5
   ) -> (Pose) {
    alphaBoolBlend {blendOption=Linear}
    boneToModify {boneName=thigh_l_ol}
    constraintSetup [{transformType=Rotation, targetBone {boneName=pelvis}}, {transformType=Rotation, targetBone {boneName=thigh_l}}]
    constraintWeights [1,1]
   }
   n2 = "Constraint - dyn_bootTabMstr_l (PPPP)"!(
    ComponentPose=@n1
    Alpha=1
    ConstraintWeights_0=0.0442
    ConstraintWeights_1:float=0.7206
    ConstraintWeights_2:float=0.1405
    ConstraintWeights_3:float=0.0948
   ) -> (Pose) {
    alphaBoolBlend {blendOption=Linear}
    boneToModify {boneName=dyn_bootTabMstr_l}
    constraintSetup [
     {targetBone {boneName=calf_l}}
     {targetBone {boneName=calf_twist_01_l}}
     {targetBone {boneName=foot_l}}
     {targetBone {boneName=calf_twist_02_l}}
    ]
    constraintWeights [1,1,1,1]
   }
   n3 = "Constraint - dyn_bootTabMstr_r (PPPP)"!(
    ComponentPose=@n2
    Alpha=1
    ConstraintWeights_0=0.0948
    ConstraintWeights_1:float=0.7206
    ConstraintWeights_2:float=0.1405
    ConstraintWeights_3:float=0.0442
   ) -> (Pose) {
    alphaBoolBlend {blendOption=Linear}
    boneToModify {boneName=dyn_bootTabMstr_r}
    constraintSetup [
     {targetBone {boneName=calf_twist_02_r}}
     {targetBone {boneName=calf_twist_01_r}}
     {targetBone {boneName=foot_r}}
     {targetBone {boneName=calf_r}}
    ]
    constraintWeights [1,1,1,1]
   }
   n4 = "Constraint - dyn_beltFrontMstr (PP)"!(ComponentPose=@n3, Alpha=1, ConstraintWeights_0=1, ConstraintWeights_1:float=1) -> (Pose) {
    alphaBoolBlend {blendOption=Linear}
    boneToModify {boneName=dyn_beltFrontMstr}
    constraintSetup [{targetBone {boneName=dyn_beltFront_r}}, {targetBone {boneName=dyn_beltFront_l}}]
    constraintWeights [1,1]
   }
   n5 = RigidBody(ComponentPose=@n4) {
    bClampLinearTranslationLimitToRefPose=true
    bOverrideWorldGravity=true
    overridePhysicsAsset=&/F_MED_Rave_Ninja_Physics_Leg_R
    simulationSpace=BaseBoneSpace
    alphaBoolBlend {blendOption=Linear}
    baseBoneRef {boneName=thigh_r}
    componentAppliedLinearAccClamp=<3000 3000 3000>
    componentLinearAccScale=<0.4 0.3 0.3>
    componentLinearVelScale=<0.4 0.3 0.3>
    overrideWorldGravity {z=-980}
   }
   n6 = Get Alpha() -> (Alpha:float)
   n7 = RigidBody!(ComponentPose=@n5, Alpha=@n6, OverrideWorldGravity:Vector=Get gravity_override() -> (gravity_override:Vector), ExternalForce) -> (Pose) {
    bClampLinearTranslationLimitToRefPose=true
    bOverrideWorldGravity=true
    overridePhysicsAsset=&/F_MED_Rave_Ninja_Physics_Arm
    simulationSpace=BaseBoneSpace
    alphaBoolBlend {blendOption=Linear}
    baseBoneRef {boneName=clavicle_l}
    componentAppliedLinearAccClamp=<3000 3000 3000>
    componentLinearAccScale=<0.4 0.1 0.1>
    componentLinearVelScale=<0.4 0.1 0.1>
    overrideWorldGravity {z=-980}
   }
   n8 = "Transform (Modify) Bone - Bone: dyn_chain1"(ComponentPose=@n7, Alpha=@n6, Translation="0,0,0", Rotation="5,0,0", Scale="1,1,1") {
    rotationMode=BMM_Additive
    rotationSpace=BCS_BoneSpace
    alphaBoolBlend {blendOption=Linear}
    boneToModify {boneName=dyn_chain1}
    rotation {pitch=5}
   }
   n9 = RigidBody!(ComponentPose=@n8, Alpha, ExternalForce) -> (Pose) {
    bClampLinearTranslationLimitToRefPose=true
    bOverrideWorldGravity=true
    overridePhysicsAsset=&/F_MED_Rave_Ninja_Physics_Jacket
    simulationSpace=BaseBoneSpace
    alphaBoolBlend {blendOption=Linear}
    baseBoneRef {boneName=spine_02}
    componentAppliedLinearAccClamp=<3000 3000 3000>
    componentLinearAccScale=<0.4 0.1 0.1>
    componentLinearVelScale=<0.4 0.1 0.1>
    overrideWorldGravity {z=-980}
   }
   n10 = Get bIsCrouching()
   n11 = "Transform (Modify) Bone - Bone: dyn_beltFrontMstr"!(ComponentPose=@n9, bAlphaBoolEnabled:bool=@n10, Translation="-5,0,1", Rotation="0,0,0", Scale="1,1,1") -> (Pose) {
    alphaInputType=Bool
    translationMode=BMM_Additive
    alphaBoolBlend {blendOption=Linear}
    boneToModify {boneName=dyn_beltFrontMstr}
    translation {x=-5, z=1}
   }
   n12 = "Transform (Modify) Bone - Bone: dyn_beltFrontMstr_l"!(ComponentPose=@n11, bAlphaBoolEnabled:bool=@n10, Translation="-5,0,3", Rotation="0,0,0", Scale="1,1,1") -> (Pose) {
    alphaInputType=Bool
    translationMode=BMM_Additive
    alphaBoolBlend {blendOption=Linear}
    boneToModify {boneName=dyn_beltFrontMstr_l}
    translation {x=-5, z=3}
   }
   n13 = "Transform (Modify) Bone - Bone: dyn_beltSideAMstr_l"!(ComponentPose=@n12, bAlphaBoolEnabled:bool=@n10, Translation="-5,0,4", Rotation="0,0,0", Scale="1,1,1") -> (Pose) {
    alphaInputType=Bool
    translationMode=BMM_Additive
    alphaBoolBlend {blendOption=Linear}
    boneToModify {boneName=dyn_beltSideAMstr_l}
    translation {x=-5, z=4}
   }
   n14 = "Transform (Modify) Bone - Bone: dyn_beltSideBMstr_l"!(ComponentPose=@n13, bAlphaBoolEnabled:bool=@n10, Translation="0,-2,2", Rotation="0,0,0", Scale="1,1,1") -> (Pose) {
    alphaInputType=Bool
    translationMode=BMM_Additive
    alphaBoolBlend {blendOption=Linear}
    boneToModify {boneName=dyn_beltSideBMstr_l}
    translation {y=-2, z=2}
   }
   n15 = RigidBody(ComponentPose=@n14, Alpha=Get crouching() -> (crouching:double)) {
    bClampLinearTranslationLimitToRefPose=true
    bOverrideWorldGravity=true
    overridePhysicsAsset=&/F_MED_Rave_Ninja_Physics_Belt
    simulationSpace=BaseBoneSpace
    alphaBoolBlend {blendOption=Linear}
    alphaScaleBiasClamp {bInterpResult=true, interpSpeedIncreasing=0.5}
    baseBoneRef {boneName=pelvis}
    componentAppliedLinearAccClamp=<1500 1500 1500>
    componentLinearAccScale {z=0.35}
    componentLinearVelScale {z=0.35}
    overrideWorldGravity {z=-980}
   }
   n16 = Get LOD1_selected() -> (LOD1_selected:double)
   n17 = "Transform (Modify) Bone - Bone: neck_clothes"(ComponentPose=@n15, Alpha=@n16, Translation="1.5,0,0", Rotation="0,0,0", Scale="1,1,1") {
    translationMode=BMM_Additive
    translationSpace=BCS_BoneSpace
    alphaBoolBlend {blendOption=Linear}
    boneToModify {boneName=neck_clothes}
    translation {x=1.5}
   }
   n18 = "Transform (Modify) Bone - Bone: head"(ComponentPose=@n17, Alpha=@n16, Translation="1.5,0,0", Rotation="0,0,0", Scale="1.2,1.15,1.15") {
    scaleMode=BMM_Additive
    translationMode=BMM_Additive
    translationSpace=BCS_BoneSpace
    alphaBoolBlend {blendOption=Linear}
    boneToModify {boneName=head}
    scale=<1.2 1.15 1.15>
    translation {x=1.5}
   }
   Output Pose(Result=Component To Local(ComponentPose=@n18)) {layerGroup=None}
  ]}
 EventGraph {nodes [
   Event BlueprintUpdateAnimation()
   n1 = Get CurrentLOD()
   > Set LOD1_selected(LOD1_selected:double="To Float (Boolean)"(InBool="Equal (Integer)"(A=@n1, B=1))) -> (Output_Get:double)
   > n2 = Sequence!(execute) -> (then_0, then_1, then_2:Exec, then_3:Exec, then_4:Exec)
   > n3 = Branch(Condition=Get bIsSkyDiving())
   > n4 = Branch(Condition=Get bIsParachuteOpened())
   Set gravity_override(execute=[@n3.else, @n4.then], gravity_override="0, 0,-980")
   n5 = Branch(execute=@n4.else, Condition="float > float"(A=Get PawnVelocityZ(), B=50) {is=title K2Node_CallFunction})
   > Set gravity_override(gravity_override="0, 0,400")
   Set gravity_override(execute=@n5.else, gravity_override="0, 0, 0")
   n6 = NOT Boolean(A=Get bIsParachuteOpened())
   n7 = Get bIsSkyDiving()
   Set skydiving(execute=@n2.then_1, skydiving:double="To Float (Boolean)"(InBool=AND Boolean(A=@n7, B=@n6))) -> (Output_Get:double)
   n8 = Get bIsCrouching()
   Set crouching(execute=@n2.then_2, crouching:double="To Float (Boolean)"(InBool=NOT Boolean(A=@n8))) -> (Output_Get:double)
   n9 = Branch(execute=@n2.then_3, Condition=Get bIsPlayingEmote())
   > Set Alpha(Alpha:float=Get Alpha_emote() -> (Alpha_emote:float)) -> (Output_Get:float)
   Set Alpha(execute=@n9.else, Alpha="Get Alpha_defaullt "() -> ("Alpha_defaullt ":float))
   n10 = Gravity Override State Function DEPRECATED(
    execute=@n2.then_4
    CharacterPartAnimInstance=Self-Reference()
    GameTimeInSeconds=Get Game Time in Seconds() -> (Game Time in Seconds:double)
    GravityOverrideParams=Get Gravity Override Params()
   )
   > Set gravity_override(gravity_override=@n10)
   Event OnBeginCrouching()
   > Set crouch alpha(crouch alpha:double=1) -> (Output_Get:double)
   Event OnEndCrouching()
   > Set crouch alpha(crouch alpha=0)
  ]}
]
variables [
 Alpha {type=number, default=0.8}
 "Alpha_defaullt " {type=number, default=1}
 Alpha_emote {type=number, default=0.75}
 crouch alpha {type=number}
 crouching {type=number}
 Game Time in Seconds {type=number}
 Gravity Override {type=Vector}
 Gravity Override Params {type=GravityOverrideParamsStruct, default {emoteOrMelee {gravityOverride=<0 0 -880>}}}
 gravity_override {type=Vector}
 LOD1_selected {type=number}
 skydiving {type=number}
]
