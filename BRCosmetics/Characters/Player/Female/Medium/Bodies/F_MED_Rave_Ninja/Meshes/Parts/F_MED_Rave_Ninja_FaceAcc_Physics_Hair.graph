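# Physics asset for the hair on the female medium Rave Ninja character's face accessory part, defining bodies for a short chain of hair bones.
# A small box body on dyn_hairMstr sits at the root, followed by three capsules on dyn_hair1 through dyn_hair3.
# Constraints link each capsule to the one before it, with very tight swing and twist limits that keep the chain nearly rigid.
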PhysicsAsset /BRCosmetics/Characters/Player/Female/Medium/Bodies/F_MED_Rave_Ninja/Meshes/Parts/F_MED_Rave_Ninja_FaceAcc_Physics_Hair
bodies {
 dyn_hair1 [{shapeType=Capsule, length=5, radius=2, center=<0.3998 2.0299 0.1694>, rotation=<rot 0 0 -90>}]
 dyn_hair2 [{shapeType=Capsule, length=5, radius=3, center=<0 2.0659 0.2026>, rotation=<rot 0 0 -90>}]
 dyn_hair3 [{shapeType=Capsule, length=5, radius=3, center=<0 2.0758 0>, rotation=<rot 0 0 -90>}]
 dyn_hairMstr [{shapeType=Box, extentX=1.01, extentY=1.01, extentZ=1.01}]
}
constraints [
 {bone1Name=dyn_hair1, bone2Name=dyn_hairMstr, swing1LimitDegrees=2, swing2LimitDegrees=2, twistLimitDegrees=1}
 {bone1Name=dyn_hair2, bone2Name=dyn_hair1, swing1LimitDegrees=2, swing2LimitDegrees=2, twistLimitDegrees=1}
 {bone1Name=dyn_hair3, bone2Name=dyn_hair2, swing1LimitDegrees=2, swing2LimitDegrees=2, twistLimitDegrees=1}
]
